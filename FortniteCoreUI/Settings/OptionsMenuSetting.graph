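# Slider row widget for the options menu, configured from a FortSettingInfo list item and synced with the Fort settings context.
# It maps normalized slider position to each setting's min/max range, reporting changes with the previous value.
# Visual updates drive a dynamic material's fillPercent and formatted value text, optionally showing render-scale resolution.
# It also handles hover or touch tooltips, mobile sizing, and disables certain look-sensitivity settings while controlling remote-controlled pawns.
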WidgetBlueprint /FortniteCoreUI/Settings/OptionsMenuSetting
graphs [
 ConvertSettingsValueToSliderValue {nodes [
   n1 = ConvertSettingsValueToSliderValue!() -> (then, SettingsValue:double=~) {is=title K2Node_FunctionEntry}
   n2 = Get MinSensitivityValue() -> (MinSensitivityValue:float)
   n3 = Get MaxSensitivityValue() -> (MaxSensitivityValue:float)
   n4 = Get MinSensitivityValue()
   > Return Node!(
    execute
    SliderValue:double=float / float(A=float - float(A=@n1, B=@n2) {is=title K2Node_CallFunction}, B=float - float(A=@n3, B=@n4)) {is=title K2Node_CallFunction}
   )
  ]}
 ConvertSliderValueToSettingsValue {nodes [
   n1 = ConvertSliderValueToSettingsValue!() -> (then, SliderValue:double=~) {is=title K2Node_FunctionEntry}
   n2 = float - float(A=Get MaxSensitivityValue(), B=Get MinSensitivityValue())
   > Return Node!(
    execute
    SettingsValue:double=float + float(A=float * float(A=@n1, B=@n2), B=Get MinSensitivityValue()) {is=title K2Node_CommutativeAssociativeBinaryOperator}
   )
  ]}
 EventGraph {nodes [
   "On Value Changed (Slider_Menu)"!() -> (OutputDelegate, then, Value:float) {is=title K2Node_ComponentBoundEvent}
   > Branch(Condition=GetIsEnabled!(self:Widget Object Reference) -> (ReturnValue:bool))
   > UpdateVisuals(self:Fort Options Menu Setting Object Reference)
   "On Button Clicked (ButtonTouchTooltip)"!() -> (OutputDelegate, then, Button:Common Button Legacy Object Reference) {is=title K2Node_ComponentBoundEvent}
   > n1 = Branch(Condition=TextIsEmpty(InText=Get HoverText() -> (HoverText:Text)))
   Show Tooltip(execute=@n1.else)
   Refresh() {is=title K2Node_CustomEvent}
   > n2 = GetSliderSettingValue!(
    execute
    self:Fort Settings Context Object Reference=Get FortSettingsContext!() -> (ReturnValue:Fort Settings Context Object Reference) {is=title K2Node_GetBlueprintContextEx}
    InSettingType:ESettingType Enum=Get SettingType() -> (SettingType:ESettingType Enum=None)
   ) -> (then, ReturnValue:float)
   > n3 = Set PreviousSettingValue(PreviousSettingValue:float=@n2) -> (Output_Get:float)
   > SetValue!(
    execute
    self:Slider Object Reference=Get Slider_Menu() -> (Slider_Menu:Analog Slider Object Reference)
    InValue:float=ConvertSettingsValueToSliderValue!(SettingsValue:double=@n3) -> (SliderValue:double) {is=title K2Node_CallFunction}
   ) -> (then)
   > UpdateVisuals()
   Event Construct()
   > n4 = GetDynamicMaterial(self=Get SliderBackground() -> (SliderBackground:Image Object Reference))
   > Set Slider Texture(Slider Texture:Material Instance Dynamic Object Reference=@n4) -> (Output_Get:Material Instance Dynamic Object Reference)
   > UpdateSize()
   n5 = IsMobileGame()
   > SetVisibility(
    self=Get ButtonTouchTooltip() -> (ButtonTouchTooltip:Icon Text Button Object Reference)
    InVisibility=Select(Option 0=Collapsed, Option 1=Visible, Index=@n5)
   )
   n6 = NOT Boolean(A=@n5)
   n7 = Get HideSpaceRight() -> (HideSpaceRight:bool)
   > SetVisibility(
    self=[
     Get Spacer_Right() -> (Spacer_Right:Spacer Object Reference)
     Get HorizontalBox_Right() -> (HorizontalBox_Right:Horizontal Box Object Reference)
    ]
    InVisibility=Select(Option 0=SelfHitTestInvisible, Option 1=Collapsed, Index=OR Boolean(A=@n7, B=@n6))
   )
   > SetVisibility(
    self=Get SizeBoxLabel() -> (SizeBoxLabel:Size Box Object Reference)
    InVisibility=Select(Option 0=SelfHitTestInvisible, Option 1=Collapsed, Index=Get HideLabelText() -> (HideLabelText:bool))
   )
   Event UpdateVisuals() {is=title K2Node_Event}
   > Update Visuals_BP()
   n8 = Event OnListItemObjectSet()
   > n9 = Cast To FortSettingInfo!(execute, Object=@n8.ListItemObject) -> (then, CastFailed, AsFort Setting Info:Fort Setting Info Object Reference) {is=title K2Node_DynamicCast}
   > Set SettingType(SettingType:ESettingType Enum=Get SettingType!(self:Fort Setting Info Object Reference=@n9) -> (SettingType=None)) -> (Output_Get:ESettingType Enum=None)
   > n10 = Set NameText(NameText:Text=Get OptionDisplayText!(self:Fort Setting Info Object Reference=@n9) -> (OptionDisplayText:Text)) -> (Output_Get:Text)
   > "SetText (Text)"(self=Get SliderText() -> (SliderText:Common Text Object Reference), InText=@n10)
   > Set HoverText(HoverText:Text=Get OptionHoverText!(self:Fort Setting Info Object Reference=@n9) -> (OptionHoverText:Text)) -> (Output_Get:Text)
   > HandlePawnSet()
   > Set MinIntegralDigits(MinIntegralDigits:int=Get MinIntegralDigits!(self:Fort Setting Info Object Reference=@n9) -> (MinIntegralDigits:int=0)) -> (Output_Get:int=0)
   > Set MaxIntegralDigits(MaxIntegralDigits:int=Get MaxIntegralDigits!(self:Fort Setting Info Object Reference=@n9) -> (MaxIntegralDigits:int=0)) -> (Output_Get:int=0)
   > Set MinFractionalDigits(MinFractionalDigits:int=Get MinFractionalDigits!(self:Fort Setting Info Object Reference=@n9) -> (MinFractionalDigits:int=0)) -> (Output_Get:int=0)
   > Set MaxFractionalDigits(MaxFractionalDigits:int=Get MaxFractionalDigits!(self:Fort Setting Info Object Reference=@n9) -> (MaxFractionalDigits:int=0)) -> (Output_Get:int=0)
   > Set MinSensitivityValue(MinSensitivityValue:float=Get MinSensitivityValue!(self:Fort Setting Info Object Reference=@n9) -> (MinSensitivityValue)) -> (Output_Get:float)
   > Set MaxSensitivityValue(MaxSensitivityValue:float=Get MaxSensitivityValue!(self:Fort Setting Info Object Reference=@n9) -> (MaxSensitivityValue)) -> (Output_Get:float)
   > Set RoundingMode(RoundingMode:ERoundingMode Enum=Get RoundingMode!(self:Fort Setting Info Object Reference=@n9) -> (RoundingMode:ERoundingMode Enum=HalfToEven)) -> (Output_Get:ERoundingMode Enum=HalfToEven)
   > SetStepSize!(
    execute
    self:Slider Object Reference=Get Slider_Menu()
    InValue:float=Get StepSize!(self:Fort Setting Info Object Reference=@n9) -> (StepSize:float)
   ) -> (then)
   > SetVisibility(self=Get ButtonTouchTooltip(), InVisibility=Select(Option 0=Collapsed, Option 1=Visible, Index=IsMobileGame()))
   Event OnMouseEnter()
   > "SetText (Text)"(self=Get TabTooltipText() -> (TabTooltipText:Common Text Object Reference), InText=Get HoverText())
   > SetSliderBarColor!(
    execute
    self:Slider Object Reference=Get Slider_Menu()
    InValue:Linear Color Structure=Make Linear Color!(R:float=0, G:float=0, B:float=0, A:float=0) -> (LinearColor:Linear Color Structure) {is=title K2Node_MakeStruct}
   ) -> (then)
   > n11 = Branch(Condition=TextIsEmpty(InText=Get HoverText()))
   n12 = Set Timer by Function Name!(
    execute=@n11.else
    Object:Object Reference
    FunctionName:String=Show Tooltip
    Time:float=1.3
    bLooping:bool=false
    bMaxOncePerFrame:bool=false
    InitialStartDelay:float=0
    InitialStartDelayVariance:float=0
   ) -> (then, ReturnValue:Timer Handle Structure) {is=title K2Node_CallFunction}
   > Set TooltipTimer(TooltipTimer:Timer Handle Structure=@n12) -> (Output_Get:Timer Handle Structure)
   Event OnMouseLeave!() -> (OutputDelegate, then, MouseEvent:"Pointer Event Structure (by ref)") {is=title K2Node_Event}
   > "SetText (Text)"(self=Get TabTooltipText())
   > SetSliderBarColor(self=Get Slider_Menu(), InValue=Make Linear Color(R=0, G=0, B=.1, A=0))
   > Clear and Invalidate Timer by Handle(Handle=Get TooltipTimer() -> (TooltipTimer:Timer Handle Structure))
   > Close(self:Menu Anchor Object Reference=Get MenuAnchorSettingDescription() -> (MenuAnchorSettingDescription:Menu Anchor Object Reference))
   Reroute Node()
   > Branch(Condition=GetIsEnabled())
   n13 = ConvertSliderValueToSettingsValue!(SliderValue:double=GetValue!(self:Slider Object Reference=Get Slider_Menu()) -> (ReturnValue:float)) -> (SettingsValue:double) {is=title K2Node_CallFunction}
   > SliderSettingChanged!(
    execute
    self:Fort Settings Context Object Reference=Get FortSettingsContext()
    InSettingType:ESettingType Enum=Get SettingType!() -> (SettingType=None)
    NameText:Text=Get NameText() -> (NameText:Text)
    PreviousValue:float=Get PreviousSettingValue() -> (PreviousSettingValue:float)
    NewValue:float=@n13
   ) -> (then)
   > Call Setting Value Changed!(execute:Exec, self:Options Menu Setting Object Reference, Value:float=@n13) -> (then:Exec) {is=title K2Node_CallDelegate}
   "On Controller Capture End (Slider_Menu)"() {is=title K2Node_ComponentBoundEvent}
   "On Mouse Capture End (Slider_Menu)"() {is=title K2Node_ComponentBoundEvent}
   "On Analog Capture (Slider_Menu)"!() -> (OutputDelegate, then, Value:float) {is=title K2Node_ComponentBoundEvent}
  ]}
 HandlePawnSet {nodes [
   HandlePawnSet() {is=title K2Node_FunctionEntry}
   n1 = GetOwningPlayer()
   > Set bRCPawn(bRCPawn:bool=Cast To FortRemoteControlledPawnAthena!(Object=Get Controlled Pawn!(self:Controller Object Reference=@n1) -> (ReturnValue:Pawn Object Reference) {is=title K2Node_CallFunction}) -> (AsFort Remote Controlled Pawn Athena:Fort Remote Controlled Pawn Athena Object Reference, bSuccess:bool=~) {is=title K2Node_DynamicCast}.bSuccess) -> (Output_Get:bool)
   > n2 = Switch on ESettingType!(execute, Selection:ESettingType Enum=Get SettingType() -> (SettingType=None)) -> (
    None:Exec
    Header:Exec
    WindowMode:Exec
    DisplayResolution:Exec
    FrameRateLimit:Exec
    VideoQuality:Exec
    ThreeDResolution:Exec
    ViewDistance:Exec
    Shadows:Exec
    AntiAliasing:Exec
    Textures:Exec
    Effects:Exec
    PostProcessing:Exec
    VSync:Exec
    MotionBlur:Exec
    ShowGrass:Exec
    MobileFPSType:Exec
    ShowFPS:Exec
    AllowLowPower:Exec
    AllowVideoPlayback:Exec
    AllowDynamicResolution:Exec
    RenderingAPI:Exec
    UseGPUCrashDebugging:Exec
    RegionHeader:Exec
    Language:Exec
    Region:Exec
    MouseSensitivityYaw:Exec
    MouseSensitivityPitch:Exec
    MouseSensitivityMultiplierForAircraft:Exec
    TouchDragSensitivity:Exec
    TouchLookSmoothing:Exec
    ControllerLookSensitivityYaw:Exec
    ControllerLookSensitivityPitch:Exec
    MouseTargetingMultiplier:Exec
    MouseScopedMultiplier:Exec
    GamepadTargetingMultiplier:Exec
    GamepadScopedMultiplier:Exec
    GamepadBuildingMultiplier:Exec
    GamepadEditModeMultiplier:Exec
    TouchLookAccelerationMultiplier:Exec
    TouchDragTargetingSensitivity:Exec
    TouchDragScopedSensitivity:Exec
    TouchBuildingMultiplier:Exec
    TouchEditModeMultiplier:Exec
    TouchVerticalSensitivity:Exec
    InvertPitchMouse:Exec
    InvertPitchController:Exec
    InvertYaw:Exec
    InvertPitchForMotion:Exec
    InvertPitchForAircraftPrimary:Exec
    InvertPitchForAircraftSecondary:Exec
    InvertPitchForAircraftTouch:Exec
    InvertPitchForFightersMouse:Exec
    InvertPitchForFightersGamepad:Exec
    InvertPitchForFightersTouch:Exec
    InvertYawForMotion:Exec
    GyroEnabled:Exec
    GyroYawAxis:Exec
    GyroSensitivity:Exec
    GyroTargetingSensitivity:Exec
    GyroScopedSensitivity:Exec
    GyroHarvestingToolSensitivity:Exec
    SafeZone:Exec
    AnonymousMode:Exec
    AnonymousCharacterMode:Exec
    HideOtherPlayerNames:Exec
    HiddenMatchmakingDelay:Exec
    AutoJoinGameVoiceChannel:Exec
    ShowVoiceIndicators:Exec
    HUDScale:Exec
    ShowViewerCount:Exec
    FirstPersonCamera:Exec
    PeripheralLighting:Exec
    PingPlaceDangerMarkerWhenTargeting:Exec
    ShowGlobalChat:Exec
    ConsoleUnlockedFPS:Exec
    ToggleDoubleTapAutorun:Exec
    AutorunLockZonePreference:Exec
    ToggleSprint:Exec
    SprintByDefault:Exec
    SprintCancelsReload:Exec
    TapInteract:Exec
    InWorldInteract:Exec
    ToggleTargeting:Exec
    HoldToSwapPickup:Exec
    AutoEquipBetterItems:Exec
    EquipFirstBuildingPieceWhenSwappingQuickbars:Exec
    EquipFirstBuildingPieceWhenSwappingQuickbarsAthena:Exec
    DisablePreEditsWhenPlacingBuilding:Exec
    AimAssist:Exec
    EditModeAimAssist:Exec
    TouchEdit:Exec
    EditConfirmOnRelease:Exec
    QuickEdit:Exec
    TurboBuild:Exec
    CreativeTurboDelete:Exec
    AutoChangeMaterial:Exec
    GamepadAutoRun:Exec
    CrossplayPreference:Exec
    AutoOpenDoors:Exec
    AutoPickupWeapons:Exec
    AutoPickupWeaponsConsolePC:Exec
    AutoSortConsumablesToRight:Exec
    EnableTryBuildOnFocus:Exec
    EditButtonHoldTime:Exec
    SlideButtonHoldTime:Exec
    AccessoriesHeader:Exec
    ForceFeedback:Exec
    ContextTutorial:Exec
    ReplayRecording:Exec
    ReplayRecordingLargeTeams:Exec
    ReplayRecordingCreativeMode:Exec
    UsePowerSavingMode:Exec
    ShadowPlayHighlights:Exec
    ShowTemperature:Exec
    BuildingPossession:Exec
    GammaValue:Exec
    MasterVolume:Exec
    MusicVolume:Exec
    SoundFXVolume:Exec
    DialogVolume:Exec
    VoiceChatVolume:Exec
    CinematicsVolume:Exec
    Subtitles:Exec
    Quality:Exec
    VoiceChat:Exec
    PushToTalk:Exec
    ProximityVoiceChat:Exec
    VoiceChatInputDevice:Exec
    VoiceChatOutputDevice:Exec
    AllowBackgroundAudio:Exec
    ColorBlindMode:Exec
    ColorBlindStrength:Exec
    IgnoreGamepadInput:Exec
    LockPrimaryInputMethodToMouse:Exec
    EnableRudderControl:Exec
    RudderDeadZone:Exec
    RudderMaxThrottle:Exec
    FOVMinimum:Exec
    FOVMaximum:Exec
    VisualizeSoundEffects:Exec
    VisualizeSoundEffectsHeader:Exec
    MoveStickDeadZone:Exec
    LookStickDeadZone:Exec
    LookSensitivityPreset:Exec
    LookSensitivityPresetAds:Exec
    LookBuildModeMultiplier:Exec
    LookEditModeMultiplier:Exec
    UseAdvancedOptions:Exec
    LookHorizontalSpeed:Exec
    LookVerticalSpeed:Exec
    LookHorizontalSpeedAds:Exec
    LookVerticalSpeedAds:Exec
    LookHorizontalBoostSpeed:Exec
    LookVerticalBoostSpeed:Exec
    LookBoostAccelerationTime:Exec
    LookHorizontalBoostSpeedAds:Exec
    LookVerticalBoostSpeedAds:Exec
    LookBoostAccelerationTimeAds:Exec
    InstantBoostWhenBuilding:Exec
    LookEaseTime:Exec
    LookInputCurve:Exec
    AimAssistStrength:Exec
    UseLegacyControls:Exec
    PlayerSurveysAllowed:Exec
    LocalNotifications:Exec
    FireMode:Exec
   ) {is=title K2Node_SwitchEnum}
   SetIsEnabled(
    execute=[
     @n2.ControllerLookSensitivityPitch
     @n2.ControllerLookSensitivityYaw
     @n2.MouseSensitivityPitch
     @n2.MouseSensitivityYaw
     @n2.TouchDragScopedSensitivity
     @n2.TouchDragSensitivity
     @n2.TouchDragTargetingSensitivity
    ]
    bInIsEnabled=NOT Boolean(A=Get bRCPawn() -> (bRCPawn:bool))
   )
  ]}
 Hide Tooltip {nodes [
   Hide Tooltip() {is=title K2Node_FunctionEntry}
   > Close(self=Get MenuAnchorSettingDescription())
  ]}
 OnGetUserMenuContent {nodes [
   OnGetUserMenuContent() {is=title K2Node_FunctionEntry}
   > n1 = Create Widget!(
    execute:Exec
    Class:User Widget Class Reference=/Game/UI/Tooltips/Tooltip-Basic-S.Tooltip-Basic-S_C
    OwningPlayer:Player Controller Object Reference=GetOwningPlayer()
    BodyText:Text=Get HoverText()
    HeaderText:Text=Get NameText()
    Icon Brush:Slate Brush Asset Object Reference
    Body Style:Common Text Style Class Reference=/FortUILibrary/Styles/Typography/Foundation/TextStyle-Base-S-B.TextStyle-Base-S-B_C
    Header Style:Common Text Style Class Reference=/FortUILibrary/Styles/Typography/Foundation/TextStyle-Base-S-B-40pc.TextStyle-Base-S-B-40pc_C
   ) -> (then:Exec, ReturnValue:Tooltip- Basic- S Object Reference)
   > SetRenderTranslation(self=@n1, Translation=(X=0,Y=48))
   > Return Node!(execute, ReturnValue:User Widget Object Reference=@n1)
  ]}
 SetFontSize {nodes [
   n1 = SetFontSize!() -> (then, Text:Common Text Object Reference, DefaulFontSize:int, MobileFont:int) {is=title K2Node_FunctionEntry}
   > n2 = Set TempFont(TempFont:Slate Font Info Structure=Get Font!(self:Text Object Reference=@n1) -> (Font:Slate Font Info Structure)) -> (Output_Get:Slate Font Info Structure)
   n3 = IsMobileGame()
   > n4 = Set members in Slate Font Info!(
    execute:Exec
    StructRef:"Slate Font Info Structure (by ref)"=@n2
    Size:float="To Float (Integer)"(InInt=Select(Option 0:int=@n1.DefaulFontSize, Option 1:int=@n1.MobileFont, Index=@n3) -> (ReturnValue:int))
   ) -> (then:Exec, StructOut:"Slate Font Info Structure (by ref)") {is=title K2Node_SetFieldsInStruct}
   > SetFont(self=@n1, InFontInfo=@n4)
  ]}
 SetSliderTextRenderScale {nodes [
   n1 = SetSliderTextRenderScale!() -> (then, Scale Normalized:double=~) {is=title K2Node_FunctionEntry}
   > n2 = GetGameUserSettings!(execute) -> (then, ReturnValue:Game User Settings Object Reference)
   > n3 = Branch(Condition=IsMobileGame())
   n4 = GetResolutionScaleInformation!(self:Game User Settings Object Reference=@n2) -> (CurrentScaleNormalized:float, CurrentScaleValue:float, MinScaleValue:float, MaxScaleValue:float)
   n5 = Lerp(A=@n4.MinScaleValue, B=@n4.MaxScaleValue, Alpha=@n1)
   n6 = "To Text (Float)"(
    Value=@n5
    RoundingMode=Get RoundingMode!() -> (RoundingMode=HalfToEven)
    MinimumIntegralDigits=Get MinIntegralDigits!() -> (MinIntegralDigits=0)
    MaximumIntegralDigits=Get MaxIntegralDigits!() -> (MaxIntegralDigits=0)
    MinimumFractionalDigits=Get MinFractionalDigits!() -> (MinFractionalDigits=0)
    MaximumFractionalDigits=Get MaxFractionalDigits!() -> (MaxFractionalDigits=0)
   )
   n7 = Get SliderValue() -> (SliderValue:Common Text Object Reference)
   > "SetText (Text)"(self=@n7, InText=Format Text!(Format="LOCTEXT(\"{Percent}%\")", Percent:Text=@n6) -> (Result))
   n8 = Break Int Point!(IntPoint:"Int Point Structure (by ref)"=GetDesktopResolution!(self:Game User Settings Object Reference=@n2) -> (ReturnValue:Int Point Structure)) -> (X:int, Y:int) {is=title K2Node_BreakStruct}
   n9 = Break Int Point(IntPoint=GetScreenResolution!(self:Game User Settings Object Reference=@n2) -> (ReturnValue:Int Point Structure))
   n10 = GetFullscreenMode!(self:Game User Settings Object Reference=@n2) -> (ReturnValue:EWindowMode Enum=Fullscreen)
   n11 = SelectInt(A=@n8.Y, B=@n9.Y, bPickA="Equal (Enum)"(A:EWindowMode Enum=@n10, B:EWindowMode Enum=WindowedFullscreen))
   n12 = "To Float (Integer)"(InInt=@n11)
   n13 = "To Float (Integer)"(InInt=GetMinYResolutionFor3DView!() -> (ReturnValue:int=0))
   n14 = "Max (Float)"(A="Clamp (Float)"(Value=float / float(A=@n13, B=@n12), Min=0, Max=1), B=float / float(A=@n5, B=100))
   n15 = int * float!(A:int=@n11, B:double=@n14) -> (ReturnValue:double)
   n16 = int * float(A=SelectInt(A=@n8, B=@n9, bPickA="Equal (Enum)"(A=@n10, B=WindowedFullscreen)), B=@n14)
   "SetText (Text)"(
    execute=@n3.else
    self=@n7
    InText=Format Text!(
     Format="LOCTEXT(\"{X} x {Y} ({Percent}%)\")"
     X:Text="To Text (Float)"(Value=@n16, bUseGrouping=false, MinimumIntegralDigits=1, MaximumIntegralDigits=324, MinimumFractionalDigits=0, MaximumFractionalDigits=0)
     Y:Text="To Text (Float)"(Value=@n15, bUseGrouping=false, MinimumIntegralDigits=1, MaximumIntegralDigits=324, MinimumFractionalDigits=0, MaximumFractionalDigits=0)
     Percent=@n6
    ) -> (Result)
   )
  ]}
 Show Tooltip {nodes [
   Show Tooltip() {is=title K2Node_FunctionEntry}
   > Open!(execute, self:Menu Anchor Object Reference=Get MenuAnchorSettingDescription(), bFocusMenu:bool=false) -> (then)
  ]}
 Update Visuals_BP {nodes [
   Update Visuals_BP() {is=title K2Node_FunctionEntry}
   > n1 = Branch(Condition=Get bDisplayRenderScale() -> (bDisplayRenderScale:bool))
   n2 = ConvertSliderValueToSettingsValue(SliderValue=GetValue(self=Get Slider_Menu()))
   > n3 = SetSliderTextRenderScale!(execute, self, Scale Normalized:double=@n2) -> (then) {is=title K2Node_CallFunction}
   UpdateSliderTexture(execute=[@n3.then, @n4.then])
   n4 = "SetText (Text)"(
    execute=@n1.else
    self=Get SliderValue()
    InText="To Text (Float)"(
     Value=@n2
     RoundingMode=Get RoundingMode() -> (RoundingMode=HalfToEven)
     MinimumIntegralDigits=Get MinIntegralDigits() -> (MinIntegralDigits=0)
     MaximumIntegralDigits=Get MaxIntegralDigits() -> (MaxIntegralDigits=0)
     MinimumFractionalDigits=Get MinFractionalDigits() -> (MinFractionalDigits=0)
     MaximumFractionalDigits=Get MaxFractionalDigits() -> (MaxFractionalDigits=0)
    )
   )
  ]}
 UpdateSize {nodes [
   UpdateSize() {is=title K2Node_FunctionEntry}
   n1 = IsMobileGame()
   > SetHeightOverride(
    self=Get SizeBoxContainer() -> (SizeBoxContainer:Size Box Object Reference)
    InHeightOverride=Select(Option 0:double=48, Option 1:double=100, Index=@n1) -> (ReturnValue:double)
   )
   > SetSize!(
    execute
    self:Spacer Object Reference=Get SpacerBottom() -> (SpacerBottom:Spacer Object Reference)
    InSize:Vector 2D Structure=Select(Option 0:Vector 2D Structure=(X=1,Y=24), Option 1:Vector 2D Structure=(X=1,Y=50), Index=@n1) -> (ReturnValue:Vector 2D Structure)
   ) -> (then)
   n2 = Get SliderValue()
   n3 = Get Font(self=@n2)
   > SetFontSize!(
    execute
    self
    Text:Common Text Object Reference=@n2
    DefaulFontSize:int=Truncate(A=Break Slate Font Info!(SlateFontInfo:"Slate Font Info Structure (by ref)"=@n3) -> (Size:float=~, LetterSpacing:int, SkewAmount:float=~, bForceMonospaced:bool=~, bMaterialIsStencil:bool=~, MonospacedWidth:float=~) {is=title K2Node_BreakStruct})
    MobileFont:int=Get FontSizeMobile() -> (FontSizeMobile:int=0)
   ) -> (then) {is=title K2Node_CallFunction}
   n4 = Get SliderText()
   n5 = Get Font(self=@n4)
   > SetFontSize(
    Text=@n4
    DefaulFontSize=Truncate(A=Break Slate Font Info(SlateFontInfo=@n5) -> (Size=~, SkewAmount=~, bForceMonospaced=~, bMaterialIsStencil=~, MonospacedWidth=~))
    MobileFont=Get FontSizeMobile() -> (FontSizeMobile=0)
   )
   n6 = Get IconBrush!(self:Icon Text Button Object Reference=Get ButtonTouchTooltip()) -> (IconBrush:Slate Brush Structure)
   > n7 = Set members in Slate Brush!(
    execute:Exec
    StructRef:"Slate Brush Structure (by ref)"=@n6
    ImageSize:Vector 2D Structure=Select(
     Option 0=Break Slate Brush!(SlateBrush:"Slate Brush Structure (by ref)"=@n6) -> (ImageSize:Vector 2D Structure, OutlineSettings:Slate Brush Outline Settings Structure) {is=title K2Node_BreakStruct}
     Option 1=Get ToolTipSizeMobile() -> (ToolTipSizeMobile:Vector 2D Structure)
     Index=IsMobileGame()
    )
   ) -> (then:Exec, StructOut:"Slate Brush Structure (by ref)") {is=title K2Node_SetFieldsInStruct}
   > Set Icon!(execute, self:Icon Text Button Object Reference=Get ButtonTouchTooltip(), IconBrush:Slate Brush Structure=@n7) -> (then) {is=title K2Node_CallFunction}
  ]}
 UpdateSliderTexture {nodes [
   UpdateSliderTexture() {is=title K2Node_FunctionEntry}
   > SetScalarParameterValue(
    self=Get Slider Texture() -> (Slider Texture:Material Instance Dynamic Object Reference)
    ParameterName=MakeLiteralName(Value=fillPercent)
    Value=GetValue(self=Get Slider_Menu())
   )
  ]}
]
variables [
 RoundingMode {type=ERoundingMode, default=HalfToEven}
 Slider Texture {type=MaterialInstanceDynamic}
 ToolTipSizeMobile {type=Vector2D, default=<60 60>}
]
widgets {
 info {inheritedWidgetCount=1, parentClass=/Script/FortniteUI.FortOptionsMenuSetting, rootWidgetClass=/Script/UMG.MenuAnchor, widgetCount=15}
 widgets [
  MenuAnchorSettingDescription {
   widgetClassPath=&MenuAnchor
   bIsVariable=true
   widget=MenuAnchorSettingDescription
   properties {menuClass=&Tooltip-Basic-S_C, placement=MenuPlacement_MenuRight}
  }
  VerticalBox_0 {widgetClassPath=&VerticalBox, parent=MenuAnchorSettingDescription, widget=VerticalBox_0, slot {class=PanelSlot}}
  HorizontalBox_1 {
   widgetClassPath=&HorizontalBox
   parent=VerticalBox_0
   widget=HorizontalBox_1
   slot {class=VerticalBoxSlot, properties {size {sizeRule=Automatic, value=0.8}}}
  }
  SizeBoxLabel {
   widgetClassPath=&SizeBox
   bIsVariable=true
   parent=HorizontalBox_1
   widget=SizeBoxLabel
   properties {bOverride_WidthOverride=true, minDesiredWidth=400, widthOverride=400}
   slot {class=HorizontalBoxSlot, properties {padding=<ltrb 0 0 20 0>, size {sizeRule=Automatic, value=0.2}}}
  }
  SliderText {
   widgetClassPath=&CommonTextBlock
   bIsVariable=true
   parent=SizeBoxLabel
   widget=SliderText
   properties {
    scrollStyle=&TextScrollStyle-Base_C
    style=/FortUILibrary/Styles/Typography/Foundation/TextStyle-Base-S-B-Blue.TextStyle-Base-S-B-Blue_C
    text=Text Block
    colorAndOpacity {colorUseRule=UseColor_Specified, specifiedColor=<rgba 0.008 0.011 0.0262 1>}
    font {
     bForceMonospaced=false
     bMaterialIsStencil=false
     fontMaterial=null
     fontObject=&NotoSans
     letterSpacing=0
     monospacedWidth=1
     size=18
     skewAmount=0
     typefaceFontName=NotoSans-Bold
     outlineSettings {
      bApplyOutlineToDropShadows=false
      bMiteredCorners=false
      bSeparateFillAlpha=false
      outlineMaterial=null
      outlineSize=0
      outlineColor=<rgba 0 0 0 1>
     }
    }
    shadowOffset=<0 0>
   }
   slot {class=SizeBoxSlot, properties {horizontalAlignment=HAlign_Right, verticalAlignment=VAlign_Center}}
  }
  SizeBoxContainer {
   widgetClassPath=&SizeBox
   bIsVariable=true
   parent=HorizontalBox_1
   widget=SizeBoxContainer
   properties {
    bOverride_HeightOverride=true
    bOverride_MinDesiredWidth=true
    heightOverride=48
    minDesiredWidth=450
    navigation=&self:WidgetTree.SizeBoxContainer.WidgetNavigation_1
   }
   slot {class=HorizontalBoxSlot, properties {size {sizeRule=Fill, value=0.6}}}
  }
  Button_0 {
   widgetClassPath=&Button
   parent=SizeBoxContainer
   widget=Button_0
   properties {
    visibility=SelfHitTestInvisible
    widgetStyle {
     clickedSlateSound {resourceObject=null}
     disabled {
      bIsDynamicallyLoaded=false
      drawAs=NoDrawType
      imageType=NoImage
      mirroring=NoMirror
      resourceName=None
      resourceObject=null
      tiling=NoTile
      imageSize=<0 0>
      margin=<ltrb 0 0 0 0>
      outlineSettings {
       bUseBrushTransparency=false
       roundingType=HalfHeightRadius
       width=0
       color {colorUseRule=UseColor_Specified, specifiedColor=<rgba 0 0 0 0>}
       cornerRadii=<0 0 0 0>
      }
      tintColor {colorUseRule=UseColor_Specified, specifiedColor=<rgba 1 1 1 0>}
      uVRegion {bIsValid=false, max=<0 0>, min=<0 0>}
     }
     disabledForeground {colorUseRule=UseColor_Specified, specifiedColor=<rgba 0.5271 0.5271 0.5271 1>}
     hovered {
      bIsDynamicallyLoaded=false
      drawAs=RoundedBox
      imageType=NoImage
      mirroring=NoMirror
      resourceName=None
      resourceObject=/FortUILibrary/Foundation/Buttons/Materials/M_Button_Settings_Select
      tiling=NoTile
      imageSize=<8 8>
      margin=<ltrb 0.49 0.49 0.49 0.49>
      outlineSettings {
       bUseBrushTransparency=true
       roundingType=FixedRadius
       width=1
       color {colorUseRule=UseColor_Specified, specifiedColor=<rgba 0.7243 0.7243 0.7243 1>}
       cornerRadii=<4 4 4 4>
      }
      tintColor {colorUseRule=UseColor_Specified, specifiedColor=<rgba 1 0.9167 0 1>}
      uVRegion {bIsValid=false, max=<0 0>, min=<0 0>}
     }
     hoveredForeground {colorUseRule=UseColor_Specified, specifiedColor=<rgba 0 1 1 1>}
     hoveredSlateSound {resourceObject=null}
     normal {
      bIsDynamicallyLoaded=false
      drawAs=RoundedBox
      imageType=NoImage
      mirroring=NoMirror
      resourceName=None
      resourceObject=null
      tiling=NoTile
      imageSize=<32 32>
      margin=<ltrb 0 0 0 0>
      outlineSettings {
       bUseBrushTransparency=true
       roundingType=FixedRadius
       width=1
       color {colorUseRule=UseColor_Specified, specifiedColor=<rgba 0.6951 0.6951 0.6951 1>}
       cornerRadii=<4 4 4 4>
      }
      tintColor {colorUseRule=UseColor_Specified, specifiedColor=<rgba 1 1 1 0>}
      uVRegion {bIsValid=false, max=<0 0>, min=<0 0>}
     }
     normalForeground {colorUseRule=UseColor_Specified, specifiedColor=<rgba 0 1 1 1>}
     normalPadding=<ltrb 4 4 4 4>
     pressed {
      bIsDynamicallyLoaded=false
      drawAs=RoundedBox
      imageType=NoImage
      mirroring=NoMirror
      resourceName=None
      resourceObject=null
      tiling=NoTile
      imageSize=<32 32>
      margin=<ltrb 0 0 0 0>
      outlineSettings {
       bUseBrushTransparency=true
       roundingType=FixedRadius
       width=1
       color {colorUseRule=UseColor_Specified, specifiedColor=<rgba 0.7243 0.7243 0.7243 1>}
       cornerRadii=<4 4 4 4>
      }
      tintColor {colorUseRule=UseColor_Specified, specifiedColor=<rgba 1 1 1 0>}
      uVRegion {bIsValid=false, max=<0 0>, min=<0 0>}
     }
     pressedForeground {colorUseRule=UseColor_Specified, specifiedColor=<rgba 0 1 1 1>}
     pressedPadding=<ltrb 4 4 4 4>
     pressedSlateSound {resourceObject=null}
    }
   }
   slot {class=SizeBoxSlot}
  }
  Overlay_0 {
   widgetClassPath=&Overlay
   parent=Button_0
   widget=Overlay_0
   slot {class=ButtonSlot, properties {horizontalAlignment=HAlign_Fill, verticalAlignment=VAlign_Fill, padding=<ltrb 0 0 0 0>}}
  }
  Slider_Menu {
   widgetClassPath=/Script/CommonUI.AnalogSlider
   bInherited=true
   bIsVariable=true
   parent=Overlay_0
   widget=Slider_Menu
   properties {sliderBarColor=<rgba 1 1 1 0>, sliderHandleColor=<rgba 1 1 1 0>}
   slot {class=OverlaySlot, properties {horizontalAlignment=HAlign_Fill, verticalAlignment=VAlign_Fill}}
  }
  SliderBackground {
   widgetClassPath=&Image
   bIsVariable=true
   parent=Overlay_0
   widget=SliderBackground
   properties {
    visibility=SelfHitTestInvisible
    brush {
     bIsDynamicallyLoaded=false
     drawAs=Image
     imageType=NoImage
     mirroring=NoMirror
     resourceName=None
     resourceObject=&/OptionsSliderFill_Light
     tiling=NoTile
     imageSize=<32 40>
     margin=<ltrb 0 0 0 0>
     outlineSettings {
      bUseBrushTransparency=false
      roundingType=HalfHeightRadius
      width=0
      color {colorUseRule=UseColor_Specified, specifiedColor=<rgba 0 0 0 0>}
      cornerRadii=<0 0 0 0>
     }
     tintColor {colorUseRule=UseColor_Specified, specifiedColor=<rgba 1 1 1 1>}
     uVRegion {bIsValid=false, max=<0 0>, min=<0 0>}
    }
   }
   slot {class=OverlaySlot, properties {horizontalAlignment=HAlign_Fill, verticalAlignment=VAlign_Fill}}
  }
  SliderValue {
   widgetClassPath=&CommonTextBlock
   bIsVariable=true
   parent=Overlay_0
   widget=SliderValue
   properties {
    style=&TextStyle-Base-S-B_C
    text=Text Block
    font {
     bForceMonospaced=false
     bMaterialIsStencil=false
     fontMaterial=null
     fontObject=&NotoSans
     letterSpacing=0
     monospacedWidth=1
     size=18
     skewAmount=0
     typefaceFontName=NotoSans-Bold
     outlineSettings {
      bApplyOutlineToDropShadows=false
      bMiteredCorners=false
      bSeparateFillAlpha=false
      outlineMaterial=null
      outlineSize=0
      outlineColor=<rgba 0 0 0 1>
     }
    }
    shadowOffset=<0 0>
   }
   slot {class=OverlaySlot, properties {verticalAlignment=VAlign_Center, padding=<ltrb 30 0 0 0>}}
  }
  ButtonTouchTooltip {
   widgetClassPath=&IconTextButton
   bIsVariable=true
   parent=HorizontalBox_1
   widget=ButtonTouchTooltip
   properties {
    buttonText=""
    onButtonBaseClicked=(&self:WidgetTree.ButtonTouchTooltip.HandleOnButtonClicked)
    onButtonBaseDoubleClicked=(&self:WidgetTree.ButtonTouchTooltip.HandleOnButtonDoubleClicked)
    onButtonBaseHovered=(&self:WidgetTree.ButtonTouchTooltip.HandleOnButtonHovered)
    onButtonBaseUnhovered=(&self:WidgetTree.ButtonTouchTooltip.HandleOnButtonUnhovered)
    onSelectedChangedBase=(&self:WidgetTree.ButtonTouchTooltip.HandleOnSelectedChanged)
    style=/FortUILibrary/Foundation/Buttons/ButtonStyles/ButtonStyle-Icon_Transparent.ButtonStyle-Icon_Transparent_C
    visibility=Collapsed
    iconBrush {
     bIsDynamicallyLoaded=false
     drawAs=Image
     imageType=NoImage
     mirroring=NoMirror
     resourceName=None
     resourceObject=/Game/UI/Icons/Icon-InfoBox-M
     tiling=NoTile
     imageSize=<40 40>
     margin=<ltrb 0 0 0 0>
     outlineSettings {
      bUseBrushTransparency=false
      roundingType=HalfHeightRadius
      width=0
      color {colorUseRule=UseColor_Specified, specifiedColor=<rgba 0 0 0 0>}
      cornerRadii=<0 0 0 1>
     }
     tintColor {colorUseRule=UseColor_Specified, specifiedColor=<rgba 1 1 1 1>}
     uVRegion {bIsValid=false, max=<0 0>, min=<0 0>}
    }
   }
   slot {class=HorizontalBoxSlot, properties {horizontalAlignment=HAlign_Left, verticalAlignment=VAlign_Center, padding=<ltrb 32 0 0 0>}}
  }
  HorizontalBox_Right {
   widgetClassPath=&HorizontalBox
   bIsVariable=true
   parent=HorizontalBox_1
   widget=HorizontalBox_Right
   slot {class=HorizontalBoxSlot, properties {size {sizeRule=Fill, value=0.2}}}
  }
  Spacer_Right {
   widgetClassPath=&Spacer
   bIsVariable=true
   parent=HorizontalBox_1
   widget=Spacer_Right
   properties {size=<230 1>}
   slot {class=HorizontalBoxSlot}
  }
  SpacerBottom {
   widgetClassPath=&Spacer
   bIsVariable=true
   parent=VerticalBox_0
   widget=SpacerBottom
   properties {size=<1 0>}
   slot {class=VerticalBoxSlot}
  }
 ]
}
